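; Railing set-piece Blueprint for the Helios RustBucket environment, with a tintable light fixture.
; Its construction script checks that the light fixture is valid, then applies the mesh color to it through the secondary-mesh color function.
; It also creates a dynamic material instance on the fixture's first material slot and sets the emissive tint from the emissive color variable.
; The event graph holds only disabled placeholder events, with no runtime logic.
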(event EventBeginPlay
  (|CallParentFunction) ; This node is disabled and will not be called.\nDrag off pins to build functionality.
)

(event Collision|EventActorBeginOverlap
  (params (OtherActor :type Actor%20Object%20Reference))
  (|CallParentFunction OtherActor) ; This node is disabled and will not be called.\nDrag off pins to build functionality.
)

(event EventTick
  (params (DeltaSeconds :type Float%20%28single-precision%29))
  (|CallParentFunction DeltaSeconds) ; This node is disabled and will not be called.\nDrag off pins to build functionality.
)

(fn ConstructionScript
  (bind _lightfixture (Variables|Default|GetLightFixture))
  (|CallParentFunction)
  (if (Utilities|IsValid (Variables|Default|GetLightFixture) :!class /Script/Engine.KismetSystemLibrary)
    (Default|SetColoronSecondaryMesh (Variables|Default|GetMeshColor) _lightfixture)
    (bind _returnvalue (Rendering|Material|CreateDynamicMaterialInstance _lightfixture 0 (Rendering|Material|GetMaterial _lightfixture 0) "None" :!class /Script/Engine.PrimitiveComponent))
    (Rendering|Material|SetVectorParameterValue _returnvalue "SFX_Emissive_Tint(Expand)" (Variables|Default|GetEmissiveColor) :!class /Script/Engine.MaterialInstanceDynamic)))
